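; AGV_LimeRock is an audio gameplay volume Blueprint that applies a sound mix modifier while the LimeRock volume is toggled on.
; OnToggledOn pushes the mix held in the Event Sound Mix Modifier variable, and OnToggledOff pops it.
; Both paths run only where cosmetic events can execute, and each prints an entered or exited debug message.
; BeginPlay, Tick, overlap and the construction script are empty.
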(event EventBeginPlay)

(event Collision|EventActorBeginOverlap
  (params (OtherActor :type Actor%20Object%20Reference)))

(event EventTick
  (params (DeltaSeconds :type Float%20%28single-precision%29)))

(event OnToggledOn(AudioGameplayVolume)
  (Networking|CanExecuteCosmeticEvents
    (:True
      (Audio|PushSoundMixModifier (Variables|Default|GetEvent%20Sound%20Mix%20Modifier))
      (Development|PrintString "EnteredLimeRockAGV" true true "(R=0.000000,G=0.660000,B=1.000000,A=1.000000)" 5.0))))

(event OnToggledOff(AudioGameplayVolume)
  (Networking|CanExecuteCosmeticEvents
    (:True
      (Audio|PopSoundMixModifier (Variables|Default|GetEvent%20Sound%20Mix%20Modifier))
      (Development|PrintString "ExitedLimeRockAGV" true true "(R=0.000000,G=0.660000,B=1.000000,A=1.000000)" 5.0))))

(fn ConstructionScript)
