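; LimeRock audio manager actor that extends a parent audio manager to control music cues and sound mix modifiers.
; On BeginPlay it moves its root to a fixed world location and pops any event-end mix.
; Delta-gated setters act only on change, sending cue numbers as named music encounter events through the global gameplay event router or pushing and popping mix modifiers.
; EndPlay switches both mix modifiers off.
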(event EventBeginPlay
  (Transformation|SetWorldLocation (Variables|Default|GetDefaultSceneRoot) "-94938.596449,-94602.274149,1457.936565" false false)
  (|CallParentFunction)
  (bind _eventendmixmod (Variables|Default|GetEventEndMixMod)
    (:then
      (Audio|PopSoundMixModifier _eventendmixmod))
    (:else)))

(event Custom|SetMusicCue
  (params (MusicCue :type Integer))
  (bind _updated_variable (Utilities|FlowControl|DeltaGateWildcard (Variables|Default|GetMusicCue) MusicCue)
    (:Success
      (MusicEncounter|ExecuteMusicEncounterGameplayEventByName (GameplayEventRouter|GetGlobalEventRouter) (Utilities|String|StringToName (Utilities|String|ToString(Integer) _updated_variable)) self))))

(event EventSetEventMix
  (params (EventMix :type Boolean))
  (|CallParentFunction EventMix))

(event Custom|SetMixModActive
  (params (MixModActive :type Boolean))
  (bind _updated_variable (Utilities|FlowControl|DeltaGateWildcard (Variables|Default|GetMixModActive) MixModActive)
    (:Success
      (if _updated_variable
        (bind _soundmixmodifier (Variables|Default|GetSoundMixModifier)
          (:then
            (Audio|PushSoundMixModifier _soundmixmodifier))
          (:else))
        (else
          (bind _soundmixmodifier_2 (Variables|Default|GetSoundMixModifier)
            (:then
              (Audio|PopSoundMixModifier _soundmixmodifier_2))
            (:else)))))))

(event Custom|SetEventEndMixActive
  (params (EventEndMixActive :type Boolean))
  (bind _updated_variable (Utilities|FlowControl|DeltaGateWildcard (Variables|Default|GetEventEndMixActive) EventEndMixActive)
    (:Success
      (if _updated_variable
        (bind _eventendmixmod (Variables|Default|GetEventEndMixMod)
          (:then
            (Audio|PushSoundMixModifier _eventendmixmod))
          (:else))
        (else
          (bind _eventendmixmod_2 (Variables|Default|GetEventEndMixMod)
            (:then
              (Audio|PopSoundMixModifier _eventendmixmod_2))
            (:else)))))))

(event EventEndPlay
  (params (EndPlayReason :type EEndPlayReason%20Enum :default "Destroyed"))
  (|CallParentFunction EndPlayReason)
  (CallFunction|SetMixModActive false)
  (CallFunction|SetEventEndMixActive false))

(fn ConstructionScript
  (|CallParentFunction))
