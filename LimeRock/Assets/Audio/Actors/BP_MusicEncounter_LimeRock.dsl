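; BP_MusicEncounter_LimeRock drives the music phases of the LimeRock encounter: event start, Rift Device, demon appearance and TV activation.
; A named-event switch calls one custom event per phase, each passing its sound to InitiateNewMusicPhase.
; When TurnOnMusic is set, that function fires a FadeOut trigger on any stored audio component, then spawns and stores the new music.
; Music spawns at a location set through SetMusicPlayLocation.
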(event Custom|PlayEventStartMusic
  (CallFunction|InitiateNewMusicPhase (Variables|MusicEncounter|GetEventStart)))

(event Custom|FireRiftDevice
  (CallFunction|InitiateNewMusicPhase (Variables|MusicEncounter|GetRiftDevice)))

(event Custom|PlayDemonAppear
  (CallFunction|InitiateNewMusicPhase (Variables|MusicEncounter|GetDemonEmerge)))

(event Custom|PlayTVActivation
  (CallFunction|InitiateNewMusicPhase (Variables|MusicEncounter|GetTVActivation)))

(event MusicEncounter|EventExecuteNamedEvent
  (params (Music%20Encounter%20Component :type Music%20Encounter%20Component%20Base%20Object%20Reference) (Event%20Name :type Name :default "None"))
  (switch name Event%20Name
    (:1
      (CallFunction|PlayEventStartMusic))
    (:2
      (CallFunction|FireRiftDevice))
    (:3
      (CallFunction|PlayDemonAppear))
    (:4
      (CallFunction|PlayTVActivation))))

(fn InitiateNewMusicPhase
  (params (Music :type Sound%20Base%20Object%20Reference))
  (Variables|Default|GetTurnOnMusic
    (:then
      (bind _limerockmusicencounteraudio (Variables|Default|GetLimeRockMusicEncounterAudio)
        (:then
          (Audio|Parameter|ExecuteTriggerParameter _limerockmusicencounteraudio "FadeOut"))
        (:else
          _))
      (bind _audio_component (MusicEncounter|SpawnMusicAtLocation Music (Variables|Default|GetMusicPlayLocation)))
      (Variables|Default|SetLimeRockMusicEncounterAudio _audio_component))
    (:else)))

(fn SetMusicPlayLocation
  (params (MusicPlayLocation :type Vector))
  (Variables|Default|SetMusicPlayLocation MusicPlayLocation))
